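# Unlit masked material for the TigerRootFame backpack's headphone FX, reacting to the louder of event and low emote audio.
# Audio fades in a dithered opacity mask from pre-skinned X position, though its bounds divisor subtracts Max from itself.
# The diffuse texture drives emissive, blended toward a reactive multiplier by an audio-sampled curve and scaled by time of day.
# Specular, normal and mask textures are sampled but unconnected.
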Material /BRCosmetics/Accessories/FORT_Backpacks/Backpack_F_MED_TigerRootFame/FX/M_Backpack_TigerRootFame_Headphone
graph {
 properties {
  bUsedWithMeshDeformer=true
  bUsedWithSkeletalMesh=true
  blendMode=BLEND_Masked
  pixelDepthOffsetMode=PDOM_Legacy
  shadingModel=MSM_Unlit
  usedShadingModels=Unlit
 }
 nodes [
  n1 = PreSkinnedLocalBounds()
  n2 = Divide(A=PreSkinnedPosition(), B=Subtract(A=@n1.Max, B=@n1.Max))
  n3 = Abs(None=VertexInterpolator(VS=MaterialFunctionCall(Float3=@n2) {materialFunction=&BreakOutFloat3Components}.R).PS)
  n4 = ScalarParameter() {defaultValue=0.17, parameterName=Width}
  n5 = Saturate(None=Multiply(A=Subtract(A=@n3, B=@n4), B=ScalarParameter() {defaultValue=10, parameterName=Mult}))
  n6 = ScalarParameter() {defaultValue=1, parameterName=Softness}
  n7 = CollectionParameter() {collection=&FortniteMaterialParameters, parameterName=EventAudio_Amplitude}
  n8 = CollectionParameter() {collection=&FortniteMaterialParameters, parameterName=Emote_AudioLow}
  n9 = Max(A=@n7, B=@n8)
  n10 = Constant()
  n11 = Constant() {r=0.4}
  MaterialFunctionCall(Alpha Threshold=LinearInterpolate(
    B=Power(Base=@n5, Exp=@n6) {constExponent=1}
    Alpha=MaterialFunctionCall(Input=@n9, Input Low=@n10, Input High=@n11) {materialFunction=&RemapValue_0-1}.Result
   ) {constA=1, constAlpha=0}) {materialFunction=&DitherTemporalAA}
  n12 = Desaturation(
   None=Multiply(
    A=TextureSample() {texture=/BRCosmetics/Accessories/FORT_Backpacks/Backpack_F_MED_TigerRootFame/Textures/T_TigerRootFame_Backpack_D}.RGB
    B=ScalarParameter() {defaultValue=0.6, parameterName=Emissive Mult}
   )
   Fraction=ScalarParameter() {defaultValue=-0.2, parameterName=Fraction}
  )
  n13 = Multiply(A=@n12, B=ScalarParameter() {defaultValue=1, parameterName=Emissive Reactive})
  n14 = CurveAtlasRowParameter(CurveTime=@n9) {
   atlas=/Game/Athena/LinearColorCurves/A_Atlas_Turtleneck_Glider
   curve=/Game/Accessories/FORT_Backpacks/Backpack_F_MED_TigerRootFame/FX/Curve_Backpack_TigerRootFame_01
   defaultValue=57
   group=Curve
   parameterName=Curve01
  }
  n15 = ScalarParameter() {defaultValue=1, parameterName=TOD_Morning}
  n16 = ScalarParameter() {defaultValue=1, parameterName=TOD_Day}
  n17 = ScalarParameter() {defaultValue=1, parameterName=TOD_Evening}
  n18 = ScalarParameter() {defaultValue=0.65, parameterName=TOD_Night}
  MaterialFunctionCall(Emissive=Multiply(
    A=LinearInterpolate(A=@n12, B=@n13, Alpha=@n14.B) {constA=1, constAlpha=0}
    B=MaterialFunctionCall(Value_Morning=@n15, Value_Day=@n16, Value_Evening=@n17, Value_Night=@n18) {materialFunction=&Scalar_Time_of_Day}.Result
   )) {materialFunction=&MF_FX_EmissiveAdjustments}
  TextureSample() {
   samplerType=SAMPLERTYPE_Masks
   texture=/BRCosmetics/Accessories/FORT_Backpacks/Backpack_F_MED_TigerRootFame/Textures/T_TigerRootFame_Backpack_S
  }
  TextureSample() {
   samplerType=SAMPLERTYPE_Normal
   texture=/BRCosmetics/Accessories/FORT_Backpacks/Backpack_F_MED_TigerRootFame/Textures/T_TigerRootFame_Backpack_N
  }
  TextureSample() {
   samplerType=SAMPLERTYPE_Masks
   texture=/BRCosmetics/Accessories/FORT_Backpacks/Backpack_F_MED_TigerRootFame/Textures/T_TigerRootFame_Backpack_M
  }
 ]
}
