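# Unlit, masked material for the headphone of the TigerRootFame backpack cosmetic, named as the non-reactive version and usable on skeletal meshes.
# The diffuse texture is scaled by an emissive multiplier, then run through Desaturation with a negative default fraction that boosts saturation.
# A time-of-day scalar, dimmer at night, multiplies that color before it enters MF_FX_EmissiveAdjustments.
# The S, N and M texture samples feed no other node shown.
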Material /BRCosmetics/Accessories/FORT_Backpacks/Backpack_F_MED_TigerRootFame/FX/M_Backpack_TigerRootFame_Headphone_NoReactive
graph {
 properties {
  bUsedWithMeshDeformer=true
  bUsedWithSkeletalMesh=true
  blendMode=BLEND_Masked
  pixelDepthOffsetMode=PDOM_Legacy
  shadingModel=MSM_Unlit
  usedShadingModels=Unlit
 }
 nodes [
  Constant() {r=1}
  n1 = Multiply(
   A=TextureSample() {texture=/BRCosmetics/Accessories/FORT_Backpacks/Backpack_F_MED_TigerRootFame/Textures/T_TigerRootFame_Backpack_D}.RGB
   B=ScalarParameter() {defaultValue=0.6, parameterName=Emissive Mult}
  )
  n2 = ScalarParameter() {defaultValue=-0.2, parameterName=Fraction}
  n3 = ScalarParameter() {defaultValue=1, parameterName=TOD_Morning}
  n4 = ScalarParameter() {defaultValue=1, parameterName=TOD_Day}
  n5 = ScalarParameter() {defaultValue=1, parameterName=TOD_Evening}
  n6 = ScalarParameter() {defaultValue=0.65, parameterName=TOD_Night}
  MaterialFunctionCall(Emissive=Multiply(
    A=Desaturation(None=@n1, Fraction=@n2)
    B=MaterialFunctionCall(Value_Morning=@n3, Value_Day=@n4, Value_Evening=@n5, Value_Night=@n6) {materialFunction=&Scalar_Time_of_Day}.Result
   )) {materialFunction=&MF_FX_EmissiveAdjustments}
  TextureSample() {
   samplerType=SAMPLERTYPE_Masks
   texture=/BRCosmetics/Accessories/FORT_Backpacks/Backpack_F_MED_TigerRootFame/Textures/T_TigerRootFame_Backpack_S
  }
  TextureSample() {
   samplerType=SAMPLERTYPE_Normal
   texture=/BRCosmetics/Accessories/FORT_Backpacks/Backpack_F_MED_TigerRootFame/Textures/T_TigerRootFame_Backpack_N
  }
  TextureSample() {
   samplerType=SAMPLERTYPE_Masks
   texture=/BRCosmetics/Accessories/FORT_Backpacks/Backpack_F_MED_TigerRootFame/Textures/T_TigerRootFame_Backpack_M
  }
 ]
}
