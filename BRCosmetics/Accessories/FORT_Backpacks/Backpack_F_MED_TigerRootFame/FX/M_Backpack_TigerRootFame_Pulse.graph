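# Translucent, unlit, two-sided Niagara mesh-particle material for the TigerRootFame backpack's pulse trail effect.
# Opacity combines a panned trail texture and sine edge mask with optional screen-space dots, world sphere mask, view-angle and depth fades.
# Emissive blends noise-tinted particle color or curve-atlas gradients, scaled by time of day, with an optional event-audio-amplitude boost.
# World position offset adds masked arch and dual sine waves, driven by dynamic parameters, plus optional spike displacement.
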Material /BRCosmetics/Accessories/FORT_Backpacks/Backpack_F_MED_TigerRootFame/FX/M_Backpack_TigerRootFame_Pulse
graph {
 properties {
  bUsedWithNiagaraMeshParticles=true
  blendMode=BLEND_Translucent
  pixelDepthOffsetMode=PDOM_Legacy
  shadingModel=MSM_Unlit
  twoSided=true
  usedShadingModels=Unlit
 }
 nodes [
  n1 = ParticleColor()
  n2 = AppendVector(A=ScalarParameter() {defaultValue=0.3, parameterName=TrailTex_OffsetU}, B=Constant())
  n3 = VectorParameter() {group=Trail, parameterName=Trail UV, defaultValue=<rgba 1.25 1 0 0>}
  n4 = Panner(Coordinate=Multiply(A=TextureCoordinate(), B=AppendVector(A=@n3.R, B=@n3.R)), Speed=AppendVector(A=@n3.B, B=@n3.B))
  n5 = ScalarParameter() {defaultValue=1, group=Trail, parameterName=Trail Scale U}
  n6 = ScalarParameter() {defaultValue=1, group=Trail, parameterName=Trail Scale V}
  n7 = TextureSampleParameter2D!(
   UVs=MaterialFunctionCall(UVs=Add(A=@n2, B=@n4), Texture Scale=AppendVector(A=@n5, B=@n6)) {materialFunction=&ScaleUVsByCenter}.UVs
   Level
   Apply View MipBias
  ) {
   group=Trail
   mipValueMode=TMVM_MipLevel
   parameterName=Trail Tex
   sortPriority=0
   texture=&/T_Backpack_TigerRootFame_Pulse_01
  }
  n8 = TextureCoordinate()
  n9 = AppendVector(A=ScalarParameter() {parameterName=Alpha Mask Power_OffsetX}, B=Constant())
  n10 = Multiply(A=ComponentMask(None=Add(A=@n8, B=@n9)) {r=true}) {constB=0.5}
  n11 = Multiply(
   A=ChannelMaskParameter(None=@n7.RGBA) {group=Trail, maskChannel=Green, parameterName=Trail Alpha Channel, defaultValue=<rgba 0 1 0 0>}.R
   B=Power(Base=Sine(None=@n10), Exp=ScalarParameter() {defaultValue=2, group=Alpha, parameterName=Alpha Mask Power})
  )
  n12 = ScalarParameter() {defaultValue=4, group=Alpha, parameterName=Alpha Mult}
  n13 = MaterialFunctionCall(World Position=WorldPosition().XYZ) {materialFunction=&TransformToClipSpace}
  n14 = ComponentMask(None=@n13.Buffer UVs for Scene Textures) {r=true}
  n15 = ScalarParameter() {defaultValue=140, parameterName=PointTexture_TilingU}
  n16 = ComponentMask(None=@n13.Buffer UVs for Scene Textures) {g=true}
  n17 = ScalarParameter() {defaultValue=70, parameterName=PointTexture_TilingV}
  n18 = Multiply(A=AppendVector(A=Multiply(A=@n14, B=@n15) {constB=100}, B=Multiply(A=@n16, B=@n17) {constB=50})) {constB=2}
  n19 = LinearInterpolate(
   A=Multiply(A=@n1.A, B=Multiply(A=@n11, B=@n12) {constB=10})
   B=Multiply(
    A=TextureSampleParameter2D!(UVs=@n18, Apply View MipBias) {parameterName=PointTexture, texture=/Game/Effects/Fort_Effects/Textures/Mask/T_HalfSphere_D}.R
    B=ScalarParameter() {defaultValue=1, parameterName=PointMask_Mul}
   )
   Alpha=ScalarParameter() {group=Noise, parameterName=Use_PointTexture}
  )
  n20 = TextureCoordinate()
  n21 = AppendVector(A=ScalarParameter() {parameterName=PointMask_OffsetU}, B=Constant())
  n22 = Multiply(A=ComponentMask(None=Add(A=@n20, B=@n21)) {r=true}) {constB=0.5}
  n23 = Sine(None=ComponentMask(None=TextureCoordinate() {uTiling=0.5, vTiling=0.5}) {g=true})
  n24 = Multiply(
   A=Power(Base=Sine(None=@n22), Exp=ScalarParameter() {defaultValue=10, parameterName=PointMask_R_Power}) {constExponent=10}
   B=Power(Base=Multiply(A=@n23) {constB=1.5}) {constExponent=20}
  )
  n25 = LinearInterpolate(A=@n19, B=Multiply(A=@n19, B=Saturate(None=@n24)), Alpha=ScalarParameter() {group=Noise, parameterName=Use_PointMask})
  n26 = ScalarParameter() {defaultValue=1, parameterName=WorldPosition_SphereMaskRadius}
  n27 = DynamicParameter() {defaultValue=<rgba 1 1 1 0>, paramNames [Wave Mult, Arch Mult, ColorCurve_Mul, WorldPosition_SphereMaskRadius]}
  n28 = Multiply(
   A=@n25
   B=SphereMask(A=WorldPosition().XYZ, B=ParticlePositionWS(), Radius=Add(A=@n26, B=@n27.WorldPosition_SphereMaskRadius)) {attenuationRadius=65, hardnessPercent=50}
  )
  n29 = ScalarParameter() {parameterName=Use_WorldPosition_SpherreMask}
  n30 = ScalarParameter() {parameterName=MeshFade}
  DepthFade(
   Opacity=Multiply(
    A=LinearInterpolate(A=@n25, B=@n28, Alpha=@n29)
    B=MaterialFunctionCall("FadeSensitivity (optional)"=@n30) {materialFunction=&FadeBasedOnViewAngle}.Result
   )
   FadeDistance=ScalarParameter() {parameterName=DepthFade}
  )
  n31 = DynamicParameter() {parameterIndex=1, defaultValue=<rgba 1 1 0 0>, paramNames [Yoffset, Strength, Use_ChangeColorCurve002, Param4]}
  LinearInterpolate(A=Constant(), B=@n31.Use_ChangeColorCurve002, Alpha=ScalarParameter() {parameterName=Use_Dyn_ChangeColorCurve002})
  n32 = ScalarParameter() {defaultValue=1, group=TOD, parameterName=TOD_Morning}
  n33 = ScalarParameter() {defaultValue=1, group=TOD, parameterName=TOD_Day}
  n34 = ScalarParameter() {defaultValue=1, group=TOD, parameterName=TOD_Evening}
  n35 = ScalarParameter() {defaultValue=0.65, group=TOD, parameterName=TOD_Night}
  n36 = Multiply(
   A=MaterialFunctionCall(
    In=ChannelMaskParameter(None=@n7.RGBA) {group=Trail, maskChannel=Green, parameterName=Trail Emissive Channel, defaultValue=<rgba 0 1 0 0>}.R
    Contrast=ScalarParameter() {defaultValue=-0.2, group=Emissive, parameterName=Emissive Contrast}
   ) {materialFunction=&CheapContrast}.Result
   B=ScalarParameter() {defaultValue=1, group=Emissive, parameterName=Emissive Mult}
  )
  n37 = VectorParameter() {group=Noise, parameterName=Noise UV, defaultValue=<rgba 1.5 1 0 0>}
  n38 = Panner(Coordinate=Multiply(A=TextureCoordinate(), B=AppendVector(A=@n37.R, B=@n37.R)), Speed=AppendVector(A=@n37.B, B=@n37.B))
  n39 = Multiply(
   A=VectorParameter() {parameterName=NoiseTex_Color, defaultValue=<rgba 1 1 1 1>}.RGB
   B=Desaturation(
    None=TextureSampleParameter2D(UVs=@n38) {group=Noise, parameterName=Noise Tex, texture=&T_Tile_RainbowColor}.RGB
    Fraction=ScalarParameter() {group=Noise, parameterName=Noise Desaturation}
   )
  ) {constA=10}
  n40 = Multiply(
   A=LinearInterpolate(A=@n36, B=Multiply(A=@n39, B=@n36), Alpha=ScalarParameter() {group=Noise, parameterName=Lerp Noise})
   B=Desaturation(None=@n1.RGB, Fraction=ScalarParameter() {group=Emissive, parameterName=ParticleColor Desaturation})
  )
  n41 = ScalarParameter() {defaultValue=2.5, group=Emissive, parameterName=Emissive Reactive}
  n42 = CollectionParameter() {collection=&FortniteMaterialParameters, parameterName=EventAudio_Amplitude}
  n43 = CurveAtlasRowParameter(CurveTime=@n42) {
   atlas=/Game/Athena/LinearColorCurves/A_Atlas_Turtleneck_Glider
   curve=/Game/Accessories/FORT_Backpacks/Backpack_F_MED_TigerRootFame/FX/Curve_Backpack_TigerRootFame_01
   defaultValue=57
   group=Curve
   parameterName=Curve01
  }
  n44 = StaticSwitchParameter(True=LinearInterpolate(A=@n40, B=Multiply(A=@n40, B=@n41), Alpha=@n43.B) {constA=1, constAlpha=0}, False=@n40) {defaultValue=true, group=Emissive, parameterName=Use Emissive Reactive}
  n45 = CurveAtlasRowParameter(CurveTime=ComponentMask(None=TextureCoordinate()) {r=true}) {
   atlas=/Game/Accessories/FORT_Backpacks/Backpack_F_MED_WaywardRebel/FX/CAtlas_WaywardRebel_ColorCurve
   curve=/Game/Characters/Player/Female/Medium/Bodies/F_MED_WaywardRebel/FX/CV_Character_LocalTalent_Flare_ColorCurve01
   defaultValue=9
   parameterName=Color001
  }
  n46 = CurveAtlasRowParameter(CurveTime=ComponentMask(None=TextureCoordinate()) {r=true}) {
   atlas=&CAtlas_WaywardRebel_ColorCurve
   curve=/Game/Characters/Player/Female/Medium/Bodies/F_MED_WaywardRebel/FX/CV_Character_LocalTalent_Flare_ColorCurve03
   defaultValue=12
   parameterName=Color002
  }
  n47 = ScalarParameter() {defaultValue=1, parameterName=ColorCurve_Mul}
  n48 = DynamicParameter() {defaultValue=<rgba 1 1 1 0>, paramNames [Wave Mult, Arch Mult, ColorCurve_Mul, WorldPosition_SphereMaskRadius]}
  n49 = LinearInterpolate(A=Constant() {r=1}, B=@n48.ColorCurve_Mul, Alpha=ScalarParameter() {parameterName=Use_DynamicColorCurve_Mul})
  n50 = Desaturation(
   None=Multiply(A=LinearInterpolate(A=@n45, B=@n46, Alpha=@n31.Use_ChangeColorCurve002), B=Multiply(A=@n47, B=@n49))
   Fraction=ScalarParameter() {parameterName=ColorCurve_Desaturation}
  )
  n51 = ScalarParameter() {parameterName=Use_ColorCurve}
  MaterialFunctionCall(Emissive=Multiply(
    A=MaterialFunctionCall(Value_Morning=@n32, Value_Day=@n33, Value_Evening=@n34, Value_Night=@n35) {materialFunction=&Scalar_Time_of_Day}.Result
    B=LinearInterpolate(A=@n44, B=@n50, Alpha=@n51)
   )) {materialFunction=&MF_FX_EmissiveAdjustments}
  n52 = Constant()
  n53 = TextureCoordinate()
  n54 = Sine(None=Multiply(A=ComponentMask(None=@n53) {r=true}) {constB=0.5})
  n55 = Multiply(
   A=Power(
    Base=MaterialFunctionCall(X=@n52, Y=@n52, Z=@n54) {materialFunction=&MakeFloat3}.Result
    Exp=ScalarParameter() {defaultValue=1, group=WPO, parameterName=Arch Power}
   ) {constExponent=10}
   B=ScalarParameter() {defaultValue=3, group=WPO, parameterName=Arch Intensity}
  )
  n56 = DynamicParameter() {defaultValue=<rgba 1 1 1 0>, paramNames [Wave Mult, Arch Mult, ColorCurve_Mul, WorldPosition_SphereMaskRadius]}
  n57 = Constant()
  n58 = ComponentMask(None=TextureCoordinate()) {r=true}
  n59 = Multiply(A=@n58, B=ScalarParameter() {defaultValue=5, group=WPO, parameterName=Wave A Tiling}) {constB=0.5}
  n60 = Time()
  n61 = Multiply(A=@n60, B=ScalarParameter() {defaultValue=4, group=WPO, parameterName=Wave A Speed}) {constB=3}
  n62 = LinearInterpolate(Alpha=Sine(None=Add(A=@n59, B=@n61))) {constA=-1}
  n63 = Multiply(A=@n58, B=ScalarParameter() {defaultValue=3, group=WPO, parameterName=Wave B Tiling}) {constB=0.5}
  n64 = Multiply(A=@n60, B=ScalarParameter() {defaultValue=-2, group=WPO, parameterName=Wave B Speed}) {constB=5}
  n65 = LinearInterpolate(Alpha=Sine(None=Add(A=@n63, B=@n64))) {constA=-1}
  n66 = Add(
   A=Multiply(
    A=MaterialFunctionCall(X=@n57, Y=@n57, Z=@n62) {materialFunction=&MakeFloat3}.Result
    B=ScalarParameter() {defaultValue=1, group=WPO, parameterName=Wave A Intensity}
   )
   B=Multiply(
    A=MaterialFunctionCall(X=@n57, Y=@n57, Z=@n65) {materialFunction=&MakeFloat3}.Result
    B=ScalarParameter() {defaultValue=-1, group=WPO, parameterName=Wave B Intensity}
   )
  )
  n67 = ComponentMask(None=TextureCoordinate()) {r=true}
  n68 = Sine(None=Multiply(A=@n67) {constB=0.5})
  n69 = ScalarParameter() {defaultValue=1, parameterName=WPOMaskPower}
  n70 = CurveAtlasRowParameter(CurveTime=@n67) {
   atlas=&A_Atlas_Turtleneck_Glider
   curve=&Curve_Backpack_TigerRootFame_01
   defaultValue=57
   group=Curve
   parameterName=Curve01
  }
  n71 = Multiply(
   A=Add(A=Multiply(A=@n55, B=@n56.Arch Mult) {constB=10}, B=Multiply(A=@n56.Wave Mult, B=@n66))
   B=LinearInterpolate(A=Power(Base=@n68, Exp=@n69), B=@n70.A, Alpha=ScalarParameter() {group=WPO, parameterName=Use Curve Mask, sortPriority=1})
  )
  n72 = TextureCoordinate()
  n73 = Power(Base=OneMinus(None=ComponentMask(None=@n72) {r=true}))
  n74 = Transform(None=LinearInterpolate(A=@n71, B=Multiply(A=@n71, B=@n73), Alpha=ScalarParameter() {parameterName=Use_RightLinearMask})) {transformSourceType=TRANSFORMSOURCE_Local}
  n75 = WorldPosition() {worldPositionShaderOffset=WPT_ExcludeAllShaderOffsets}
  n76 = MaterialFunctionCall(Float3=TransformPosition(None=Add(A=@n74, B=@n75.XYZ)) {transformSourceType=TRANSFORMPOSSOURCE_World, transformType=TRANSFORMPOSSOURCE_Instance}) {materialFunction=&BreakOutFloat3Components}
  n77 = DynamicParameter() {parameterIndex=1, defaultValue=<rgba 1 1 0 0>, paramNames [Yoffset, Strength, Use_ChangeColorCurve002, Param4]}
  n78 = Multiply(A=Multiply(A=@n76.G, B=Multiply(A=@n77.Strength) {constB=-1}) {constB=0}, B=Divide(A=Add(A=@n76.B, B=@n77.Yoffset), B=Constant() {r=50}))
  n79 = Transform(None=Constant3Vector() {constant=<rgba 0 1 0 0>}) {transformSourceType=TRANSFORMSOURCE_Instance}
  StaticSwitchParameter(True=Add(A=@n74, B=Multiply(A=@n78, B=@n79)), False=@n74) {group=WPO, parameterName=Use Spike WPO, sortPriority=0}
 ]
}
