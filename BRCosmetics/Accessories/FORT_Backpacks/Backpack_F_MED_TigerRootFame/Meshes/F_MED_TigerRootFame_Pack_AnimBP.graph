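# Animation Blueprint for the TigerRootFame back bling, adding springy physics and idle floating to its pack, cloud and headphones bones.
# AnimGraph runs half-alpha Anim Dynamics on pack and cloud, then additive bone transforms on headphones and cloud.
# Each update builds the floating offsets from game-time sine waves at differing frequencies per axis.
# Jump-start and fall-end events briefly raise booleans that blend in downward cloud dips, the landing one deeper.
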AnimBlueprint /BRCosmetics/Accessories/FORT_Backpacks/Backpack_F_MED_TigerRootFame/Meshes/F_MED_TigerRootFame_Pack_AnimBP
target_skeleton=&/F_MED_TigerRootFame_Pack_Skeleton
graphs [
 AnimGraph {nodes [
   n1 = Anim Dynamics - Bone: pack!(
    ComponentPose=Local To Component(LocalPose=Input Pose())
    Alpha=0.5
    GravityOverride:Vector="0,0,20"
    ComponentLinearAccScale:Vector="0.7,0.7,1.5"
    ComponentLinearVelScale:Vector="0.7,0.7,1.5"
    ExternalForce="0,0,0"
   ) -> (Pose) {
    alpha=0
    angularBiasOverride=1
    angularDampingOverride=0.75
    angularSpringConstant=20
    bGravityOverrideInSimSpace=true
    bLinearSpring=true
    bOverrideLinearDamping=true
    bUseGravityOverride=true
    bUsePlanarLimit=false
    gravityScale=0
    linearDampingOverride=0.81
    linearSpringConstant=22
    simulationSpace=BoneRelative
    alphaBoolBlend {blendOption=Linear}
    alphaScaleBiasClamp {bInterpResult=true}
    boundBone {boneName=pack}
    componentAppliedLinearAccClamp=<63 53 47>
    componentLinearAccScale=<0.7 0.7 1.5>
    componentLinearVelScale=<0.7 0.7 1.5>
    gravityOverride {z=20}
    physicsBodyDefinitions [{
      sphereCollisionRadius=0
      boundBone {boneName=pack}
      boxExtents=<22 22 22>
      constraintSetup {
       linearXLimitType=Free
       linearZLimitType=Free
       angularLimitsMax=<0 0 0>
       angularLimitsMin=<0 0 0>
       angularTarget=<0 0 0>
       linearAxesMax=<2.5 0.5 0>
       linearAxesMin=<-2.5 -10 0>
      }
      localJointOffset=<0 0 0>
     }]
    relativeSpaceBone {boneName=root}
   }
   n2 = Anim Dynamics - Bone: cloud!(
    ComponentPose=@n1
    Alpha=0.5
    GravityOverride:Vector="0,0,30"
    ComponentLinearAccScale:Vector="0.7,0.7,1.7"
    ComponentLinearVelScale:Vector="0.7,0.7,1.7"
    ExternalForce="0,0,0"
   ) -> (Pose) {
    alpha=0
    angularBiasOverride=0.8
    angularDampingOverride=0.75
    angularSpringConstant=20
    bGravityOverrideInSimSpace=true
    bLinearSpring=true
    bOverrideAngularBias=true
    bOverrideLinearDamping=true
    bUseGravityOverride=true
    bUsePlanarLimit=false
    gravityScale=0
    linearDampingOverride=0.97
    linearSpringConstant=110
    simulationSpace=RootRelative
    alphaBoolBlend {blendOption=Linear}
    alphaScaleBiasClamp {bInterpResult=true}
    boundBone {boneName=cloud}
    componentAppliedLinearAccClamp=<49 49 49>
    componentLinearAccScale=<0.7 0.7 1.5>
    componentLinearVelScale=<0.7 0.7 1.5>
    gravityOverride {z=20}
    physicsBodyDefinitions [{
      sphereCollisionRadius=0
      boundBone {boneName=cloud}
      boxExtents=<20 2 1>
      constraintSetup {
       linearXLimitType=Free
       linearYLimitType=Free
       linearZLimitType=Free
       angularLimitsMax=<0 0 0>
       angularLimitsMin=<0 0 0>
       angularTarget=<0 0 0>
       linearAxesMax=<10 5 10>
       linearAxesMin=<-10 -5 -30>
      }
      localJointOffset=<0.2 0.1 1.2>
     }]
    relativeSpaceBone {boneName=root}
   }
   n3 = "Transform (Modify) Bone - Bone: headphones"(ComponentPose=@n2, Alpha=1, Translation=Get headphones floating() -> (headphones floating:Vector), Rotation="0,0,0", Scale="1,1,1") {translationMode=BMM_Additive, translationSpace=BCS_BoneSpace, alphaBoolBlend {blendOption=Linear}, boneToModify {boneName=headphones}}
   n4 = "Transform (Modify) Bone - Bone: cloud"(ComponentPose=@n3, Alpha=1, Translation=Get cloud floating() -> (cloud floating:Vector), Rotation="0,0,0", Scale="1,1,1") {translationMode=BMM_Additive, translationSpace=BCS_BoneSpace, alphaBoolBlend {blendOption=Linear}, boneToModify {boneName=cloud}}
   n5 = "Transform (Modify) Bone - Bone: cloud"!(
    ComponentPose=@n4
    bAlphaBoolEnabled:bool=Get jumping Alpha() -> (jumping Alpha:bool)
    Translation="0,0,-0.4"
    Rotation="0,0,0"
    Scale="1,1,1"
   ) -> (Pose) {
    alphaInputType=Bool
    rotationMode=BMM_Additive
    rotationSpace=BCS_BoneSpace
    translationMode=BMM_Additive
    translationSpace=BCS_BoneSpace
    alphaBoolBlend {blendInTime=0.07, blendOutTime=0.4}
    boneToModify {boneName=cloud}
    translation {z=-0.4}
   }
   n6 = "Transform (Modify) Bone - Bone: cloud"(
    ComponentPose=@n5
    bAlphaBoolEnabled=Get jumping Alpha_2() -> (jumping Alpha_2:bool)
    Translation="0,0,-1.7"
    Rotation="0,0,0"
    Scale="1,1,1"
   ) {
    alphaInputType=Bool
    rotationMode=BMM_Additive
    rotationSpace=BCS_BoneSpace
    translationMode=BMM_Additive
    translationSpace=BCS_BoneSpace
    alphaBoolBlend {blendInTime=0.1, blendOption=Cubic, blendOutTime=0.3}
    boneToModify {boneName=cloud}
    translation {z=-1.7}
   }
   Output Pose(Result=Component To Local(ComponentPose=@n6)) {layerGroup=None}
  ]}
 EventGraph {nodes [
   Event BlueprintUpdateAnimation()
   n1 = float * float(A=GetGameTimeInSeconds(), B=1.2) {is=title K2Node_PromotableOperator}
   n2 = "Sin (Radians)"(A=float * float(A=0.6999, B=@n1) {is=title K2Node_CommutativeAssociativeBinaryOperator})
   n3 = "Sin (Radians)"(A=float * float(A=2, B=@n1))
   n4 = "Sin (Radians)"(A=float * float(A=0.6, B=@n1))
   > Set cloud floating(cloud floating:Vector=MakeVector(X=float * float(A=@n4, B=0.4), Y=float * float(A=@n3, B=0.15), Z=float * float(A=@n2, B=1.7))) -> (Output_Get:Vector)
   n5 = float * float(A=GetGameTimeInSeconds(), B=1.2) {is=title K2Node_PromotableOperator}
   n6 = "Sin (Radians)"(A=float * float(A=1.15, B=@n5) {is=title K2Node_CommutativeAssociativeBinaryOperator})
   n7 = "Sin (Radians)"(A=float * float(A=0.6, B=@n5))
   n8 = "Sin (Radians)"(A=float * float(A=0.7, B=@n5))
   > Set headphones floating(headphones floating:Vector=MakeVector(X=float * float(A=@n7, B=0.5), Y=float * float(A=@n6, B=0.15), Z=float * float(A=@n8, B=1.3))) -> (Output_Get:Vector)
   Event OnBeginJumping()
   > Set jumping Alpha(jumping Alpha:bool=true) -> (Output_Get:bool)
   > Delay(Duration=0.25)
   > Set jumping Alpha(jumping Alpha=false)
   Event OnEndFalling()
   > Set jumping Alpha_2(jumping Alpha_2:bool=true) -> (Output_Get:bool)
   > Delay(Duration=0.2)
   > Set jumping Alpha_2(jumping Alpha_2=false)
   TryGetPawnOwner()
   n9 = float * float(B=1.2) {is=title K2Node_PromotableOperator}
   n10 = float * float(A=0.8, B=@n9) {is=title K2Node_CommutativeAssociativeBinaryOperator}
   n11 = float * float(A=1.5, B=@n9)
   n12 = float * float(A=0.6, B=@n9)
   MakeVector(
    X=float * float(A="Sin (Radians)"(A=@n12), B=0.5)
    Y=float * float(A="Sin (Radians)"(A=@n11), B=0.18)
    Z=float * float(A="Sin (Radians)"(A=@n10), B=1)
   )
  ]}
]
variables [
 cloud floating {type=Vector}
 headphones floating {type=Vector}
 jumping Alpha {type=boolean}
 jumping Alpha_2 {type=boolean}
]
